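; Timed Niagara anim notify state that temporarily drives float parameters on a character's cosmetic VFX from animation curves.
; Begin gathers idle and tag-matched effect components from cosmetic part modifiers, optionally filtered by Niagara system, and caches their starting values per notify instance ID.
; Tick copies blended or unblended curve values into the mapped Niagara variables.
; End restores the cached values and removes that instance's entry, with optional debug printing throughout.
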(fn AddFXToFXSystemComponentMap
  (params (FX%20System%20Component :type FXSystem%20Component%20Object%20Reference%20%28by%20ref%29) (ID :type Integer))
  (bind _id_fx_map (Variables|Default|GetID%20FX%20Map))
  (bind (_value _returnvalue) (Utilities|Map|Find _id_fx_map ID))
  (bind _value_fxsystemcomponentmaparray_7_6f6284274a32afe5b97ef9acf3a43d47 (Utilities|Struct|BreakStrucNiagaraNotifyCosmeticFXTransitionIDFXMap _value))
  (bind _map (Utilities|Map|MakeMap "Init" 0.0))
  (bind _struc_niagaranotify_cosmeticfxtransition (Utilities|Struct|MakeStrucNiagaraNotifyCosmeticFXTransition FX%20System%20Component _map))
  (bind _temp_fx_map_array (LocalVariables|GetTemp%20FX%20Map%20Array))
  (bind _asniagara_particle_system_component (Utilities|Casting|CastToNiagaraComponent FX%20System%20Component)
    (:then
      (if (or (Variables|CharacterFXTransitions|GetApply%20To%20All%20FX%20On%20Character) (Utilities|Array|ContainsItem (Variables|CharacterFXTransitions|GetNiagara%20System%20Filter) (Niagara|GetNiagaraSystemAsset _asniagara_particle_system_component)))
        (if _returnvalue
          (Utilities|Array|AppendArray _temp_fx_map_array _value_fxsystemcomponentmaparray_7_6f6284274a32afe5b97ef9acf3a43d47)
          (Utilities|Array|Add _temp_fx_map_array _struc_niagaranotify_cosmeticfxtransition)
          (else
            (Utilities|Array|Add _temp_fx_map_array _struc_niagaranotify_cosmeticfxtransition)))
        (Utilities|Map|Add _id_fx_map ID (Utilities|Struct|MakeStrucNiagaraNotifyCosmeticFXTransitionIDFXMap (LocalVariables|GetTemp%20FX%20Map%20Array)))
        (return true)
        (else
          _)))
    (:CastFailed
      _))
  (return false))

(fn PrintDebugInfo
  (params (Debug%20String :type String) (World%20Context%20Object :type Object%20Reference))
  (if (Variables|Debug|GetPrint%20Debug)
    (Development|PrintString World%20Context%20Object Debug%20String)
    (else
      _))
  (return))

(fn Received_NotifyBegin
  (params (MeshComp :type Skeletal%20Mesh%20Component%20Object%20Reference) (Animation :type Anim%20Sequence%20Base%20Object%20Reference) (TotalDuration :type Float%20%28single-precision%29) (EventReference :type Anim%20Notify%20Event%20Reference%20Structure%20%28by%20ref%29))
  (bind _returnvalue (Components|GetOwner MeshComp))
  (bind _instanceid (LocalVariables|GetInstanceID))
  (bind _id_fx_map (Variables|Default|GetID%20FX%20Map))
  (bind (_value _returnvalue_2) (Utilities|Map|Find _id_fx_map _instanceid))
  (bind _value_fxsystemcomponentmaparray_7_6f6284274a32afe5b97ef9acf3a43d47 (Utilities|Struct|BreakStrucNiagaraNotifyCosmeticFXTransitionIDFXMap _value))
  (bind _instanceid_2 (LocalVariables|GetInstanceID))
  (bind _id_fx_map_2 (Variables|Default|GetID%20FX%20Map))
  (bind (_value_2 _returnvalue_3) (Utilities|Map|Find _id_fx_map_2 _instanceid_2))
  (bind _value_fxsystemcomponentmaparray_7_6f6284274a32afe5b97ef9acf3a43d47_2 (Utilities|Struct|BreakStrucNiagaraNotifyCosmeticFXTransitionIDFXMap _value_2))
  (bind _temp_fx_map (LocalVariables|GetTemp%20FX%20Map%20))
  (bind _returnvalue_4 (|CallParentFunction MeshComp Animation TotalDuration EventReference))
  (LocalVariables|SetInstanceID (Utilities|Struct|BreakAnimNotifyEventReference EventReference))
  (Utilities|FlowControl|Sequence
    (:then_0
      (Utilities|IsValid _returnvalue
        (:Is%20Valid
          (bind (_partmodifiers _returnvalue_5) (Cosmetics|GetCosmeticPartModifiers MeshComp _returnvalue))
          (if _returnvalue_5
            (for _array_element _partmodifiers
              (Utilities|FlowControl|Sequence
                (:then_0
                  (CallFunction|AddFXtoFXSystemComponentMap (Class|CustomCharacterPartModifier|GetIdleVFXComponent _array_element) (LocalVariables|GetInstanceID)))
                (:then_1
                  (for _array_element_2 (Variables|Default|GetEffect%20Type%20Tag)
                    (CallFunction|AddFXtoFXSystemComponentMap (Class|CustomCharacterPartModifier|GetVFXComponent _array_element _array_element_2) (LocalVariables|GetInstanceID))))))
            (if _returnvalue_2
              (bind _output_get (LocalVariables|SetTemp%20FX%20Map%20 _value_fxsystemcomponentmaparray_7_6f6284274a32afe5b97ef9acf3a43d47))
              (for _index (range (+ (Utilities|Array|LastIndex _output_get) 1))
                (bind (_output_fxsystemcomponent_2_85f6eca14fa1954b13131c913da8212b _output_namevaluemapping_7_7eb069db48d3bd6fc07e65b0aa1589a6) (Utilities|Struct|BreakStrucNiagaraNotifyCosmeticFXTransition (Utilities|Array|Get(aref) _temp_fx_map _index)))
                (Utilities|Map|Clear _output_namevaluemapping_7_7eb069db48d3bd6fc07e65b0aa1589a6)
                (bind _asniagara_particle_system_component (Utilities|Casting|CastToNiagaraComponent _output_fxsystemcomponent_2_85f6eca14fa1954b13131c913da8212b))
                (for _array_element_3 (Variables|CharacterFXTransitions|GetFX%20Transition%20Curves)
                  (bind _array_element_uservariablename_5_c3e793d64b2060301331ac805c0b1026 (Utilities|Struct|BreakAnimCurveParameterPair _array_element_3))
                  (bind (_output_fxsystemcomponent_2_85f6eca14fa1954b13131c913da8212b_2 _output_namevaluemapping_7_7eb069db48d3bd6fc07e65b0aa1589a6_2) (Utilities|Struct|BreakStrucNiagaraNotifyCosmeticFXTransition (Utilities|Array|Get(aref) _temp_fx_map _index)))
                  (Utilities|Map|Add _output_namevaluemapping_7_7eb069db48d3bd6fc07e65b0aa1589a6_2 _array_element_uservariablename_5_c3e793d64b2060301331ac805c0b1026 (.ReturnValue (Niagara|GetNiagaraVariable(Float) _asniagara_particle_system_component _array_element_uservariablename_5_c3e793d64b2060301331ac805c0b1026)))
                  (CallFunction|PrintDebugInfo (Utilities|String|Append (Utilities|GetDisplayName _asniagara_particle_system_component) " ---  " (Utilities|String|ToString(Name) _array_element_uservariablename_5_c3e793d64b2060301331ac805c0b1026) " : " (Utilities|String|ToString(Float) (.ReturnValue (Niagara|GetNiagaraVariable(Float) _asniagara_particle_system_component _array_element_uservariablename_5_c3e793d64b2060301331ac805c0b1026))) "| Notify ID: " (Utilities|String|ToString(Integer) (LocalVariables|GetInstanceID))) MeshComp))
                (Utilities|Map|Add (Variables|Default|GetID%20FX%20Map) (LocalVariables|GetInstanceID) (Utilities|Struct|MakeStrucNiagaraNotifyCosmeticFXTransitionIDFXMap (LocalVariables|GetTemp%20FX%20Map%20)))
                (CallFunction|PrintDebugInfo "---- Mapped Name / Values ----- START " MeshComp)
                (if _returnvalue_3
                  (bind (_output_fxsystemcomponent_2_85f6eca14fa1954b13131c913da8212b_3 _output_namevaluemapping_7_7eb069db48d3bd6fc07e65b0aa1589a6_3) (Utilities|Struct|BreakStrucNiagaraNotifyCosmeticFXTransition (Utilities|Array|Get(aref) _value_fxsystemcomponentmaparray_7_6f6284274a32afe5b97ef9acf3a43d47_2 _index)))
                  (bind (_keypin _valuepin) (Utilities|Map|ForEachLoop(Map) _output_namevaluemapping_7_7eb069db48d3bd6fc07e65b0aa1589a6_3)
                    (:then
                      (CallFunction|PrintDebugInfo (Utilities|String|Append (Utilities|GetDisplayName _output_fxsystemcomponent_2_85f6eca14fa1954b13131c913da8212b_3) " | " (Utilities|String|ToString(Name) _keypin) " : " (Utilities|String|ToString(Float) _valuepin) "| Notify ID :" (Utilities|String|ToString(Integer) (LocalVariables|GetInstanceID))) MeshComp))
                    (:CompletedPin
                      (CallFunction|PrintDebugInfo "---- Mapped Name / Values ----- END" MeshComp))))))
            (else
              (CallFunction|PrintDebugInfo "Failed to get Cosmetic Part Modifiers" MeshComp)
              (return false))))
        (:Is%20Not%20Valid
          (CallFunction|PrintDebugInfo (Utilities|String|Append "Notify ID: " (Utilities|String|ToString(Integer) (LocalVariables|GetInstanceID)) "  Invalid Owner") MeshComp)
          (return false))))
    (:then_1
      (if (not _returnvalue_4)
        (CallFunction|PrintDebugInfo "Parent Notify begin returned false unexpectedly but Cosemetic FX Transition logic finished as expected." MeshComp))
      (return true))))

(fn Received_NotifyEnd
  (params (MeshComp :type Skeletal%20Mesh%20Component%20Object%20Reference) (Animation :type Anim%20Sequence%20Base%20Object%20Reference) (EventReference :type Anim%20Notify%20Event%20Reference%20Structure%20%28by%20ref%29))
  (bind _notifyinstanceid (Utilities|Struct|BreakAnimNotifyEventReference EventReference))
  (bind _returnvalue (Components|GetOwner MeshComp))
  (bind _id_fx_map (Variables|Default|GetID%20FX%20Map))
  (bind _value (Utilities|Map|Find _id_fx_map _notifyinstanceid))
  (bind _value_fxsystemcomponentmaparray_7_6f6284274a32afe5b97ef9acf3a43d47 (Utilities|Struct|BreakStrucNiagaraNotifyCosmeticFXTransitionIDFXMap _value))
  (bind _returnvalue_2 (|CallParentFunction MeshComp Animation EventReference))
  (for _index (range (Utilities|Array|Length _value_fxsystemcomponentmaparray_7_6f6284274a32afe5b97ef9acf3a43d47))
    (bind (_output_fxsystemcomponent_2_85f6eca14fa1954b13131c913da8212b _output_namevaluemapping_7_7eb069db48d3bd6fc07e65b0aa1589a6) (Utilities|Struct|BreakStrucNiagaraNotifyCosmeticFXTransition (Utilities|Array|Get(aref) _value_fxsystemcomponentmaparray_7_6f6284274a32afe5b97ef9acf3a43d47 _index)))
    (bind _asniagara_particle_system_component (Utilities|Casting|CastToNiagaraComponent _output_fxsystemcomponent_2_85f6eca14fa1954b13131c913da8212b)
      (:then
        (bind (_keypin _valuepin) (Utilities|Map|ForEachLoop(Map) _output_namevaluemapping_7_7eb069db48d3bd6fc07e65b0aa1589a6))
        (Niagara|SetNiagaraVariable(Float) _asniagara_particle_system_component _keypin _valuepin)
        (CallFunction|PrintDebugInfo (Utilities|String|Append "| Final Niagara System Values" (Utilities|GetDisplayName _asniagara_particle_system_component) (Utilities|String|ToString(Name) _keypin) " : " (Utilities|String|ToString(Float) (.ReturnValue (Niagara|GetNiagaraVariable(Float) _asniagara_particle_system_component _keypin)))) _returnvalue))
      (:CastFailed
        (LocalVariables|SetEnd%20Return false))))
  (bind _returnvalue_3 (Utilities|Map|Remove (Variables|Default|GetID%20FX%20Map) _notifyinstanceid))
  (if (not _returnvalue_3)
    (CallFunction|PrintDebugInfo (Utilities|String|Append "Unable To Remove ID Data for Notify ID:" (Utilities|String|ToString(Integer) _notifyinstanceid)) _returnvalue))
  (if (not _returnvalue_2)
    (CallFunction|PrintDebugInfo "Parent Notify End returned false unexpectedly but Cosemetic FX Transition logic finished as expected." _returnvalue))
  (if (LocalVariables|GetEnd%20Return)
    (return true)
    (else
      (CallFunction|PrintDebugInfo "Tick End an FX system failed to cast properly when resetting values" _returnvalue)
      (return false))))

(fn Received_NotifyTick
  (params (MeshComp :type Skeletal%20Mesh%20Component%20Object%20Reference) (Animation :type Anim%20Sequence%20Base%20Object%20Reference) (FrameDeltaTime :type Float%20%28single-precision%29) (EventReference :type Anim%20Notify%20Event%20Reference%20Structure%20%28by%20ref%29))
  (bind _notifyinstanceid (Utilities|Struct|BreakAnimNotifyEventReference EventReference))
  (bind _returnvalue (Components|GetOwner MeshComp))
  (bind _returnvalue_2 (Components|SkeletalMesh|GetAnimInstance MeshComp))
  (bind _id_fx_map (Variables|Default|GetID%20FX%20Map))
  (bind (_value _returnvalue_3) (Utilities|Map|Find _id_fx_map _notifyinstanceid))
  (bind _value_fxsystemcomponentmaparray_7_6f6284274a32afe5b97ef9acf3a43d47 (Utilities|Struct|BreakStrucNiagaraNotifyCosmeticFXTransitionIDFXMap _value))
  (bind _returnvalue_4 (Utilities|Animation|Notifies|GetCurrentAnimationTime EventReference))
  (bind _apply_anim_blending_to_transition_curve (Variables|CharacterFXTransitions|GetApply%20Anim%20Blending%20to%20Transition%20Curve))
  (bind _returnvalue_5 (|CallParentFunction MeshComp Animation FrameDeltaTime EventReference))
  (if _returnvalue_3
    (Utilities|FlowControl|Sequence
      (:then_0
        (for _array_element (Variables|CharacterFXTransitions|GetFX%20Transition%20Curves)
          (bind (_array_element_uservariablename_5_c3e793d64b2060301331ac805c0b1026 _array_element_animcurvename_6_457883094db38a3ef423919ede4d33c1) (Utilities|Struct|BreakAnimCurveParameterPair _array_element))
          (for _index (range (Utilities|Array|Length _value_fxsystemcomponentmaparray_7_6f6284274a32afe5b97ef9acf3a43d47))
            (Utilities|IsValid _returnvalue_2
              (:Is%20Valid
                (if (.ReturnValue (Animation|Curves|GetCurveValuewithDefault _returnvalue_2 _array_element_animcurvename_6_457883094db38a3ef423919ede4d33c1 -1.0))
                  (bind _output_fxsystemcomponent_2_85f6eca14fa1954b13131c913da8212b (Utilities|Struct|BreakStrucNiagaraNotifyCosmeticFXTransition (Utilities|Array|Get(aref) _value_fxsystemcomponentmaparray_7_6f6284274a32afe5b97ef9acf3a43d47 _index)))
                  (bind _asniagara_particle_system_component (Utilities|Casting|CastToNiagaraComponent _output_fxsystemcomponent_2_85f6eca14fa1954b13131c913da8212b)
                    (:then
                      (Niagara|SetNiagaraVariable(Float) _asniagara_particle_system_component _array_element_uservariablename_5_c3e793d64b2060301331ac805c0b1026 (Math|Float|SelectFloat (Animation|Curves|GetCurveValuewithDefault _returnvalue_2 _array_element_animcurvename_6_457883094db38a3ef423919ede4d33c1 -1.0) (Animation|Curves|GetUnblendedAnimationCurveValue Animation _returnvalue_4 _array_element_animcurvename_6_457883094db38a3ef423919ede4d33c1) _apply_anim_blending_to_transition_curve))
                      (CallFunction|PrintDebugInfo (Utilities|String|Append "TICK | Anim Curve Values |" (Utilities|String|ToString(Name) _array_element_animcurvename_6_457883094db38a3ef423919ede4d33c1) " : " (Utilities|String|ToString(Float) (Math|Float|SelectFloat (Animation|Curves|GetCurveValuewithDefault _returnvalue_2 _array_element_animcurvename_6_457883094db38a3ef423919ede4d33c1 -1.0) (Animation|Curves|GetUnblendedAnimationCurveValue Animation _returnvalue_4 _array_element_animcurvename_6_457883094db38a3ef423919ede4d33c1) _apply_anim_blending_to_transition_curve)) " | Niagara Values |" (Utilities|GetDisplayName _asniagara_particle_system_component) (Utilities|String|ToString(Name) _array_element_uservariablename_5_c3e793d64b2060301331ac805c0b1026) " : " (Utilities|String|ToString(Float) (.ReturnValue (Niagara|GetNiagaraVariable(Float) _asniagara_particle_system_component _array_element_uservariablename_5_c3e793d64b2060301331ac805c0b1026)))) _returnvalue))
                    (:CastFailed
                      (LocalVariables|SetTick%20Return false)))
                  (else
                    (CallFunction|PrintDebugInfo (Utilities|String|Append "Invalid anim Curve Found : " (Utilities|String|ToString(Name) _array_element_animcurvename_6_457883094db38a3ef423919ede4d33c1) " |  " (Utilities|GetDisplayName _output_fxsystemcomponent_2_85f6eca14fa1954b13131c913da8212b)) _returnvalue))))))))
      (:then_1
        (if (not _returnvalue_5)
          (CallFunction|PrintDebugInfo "Parent Notify Tick returned false unexpectedly but Cosemetic FX Transition logic finished as expected." _returnvalue))
        (if (LocalVariables|GetTick%20Return)
          (return true)
          (else
            (CallFunction|PrintDebugInfo "Tick Niagara cast failed " _returnvalue)
            (return false)))))))
