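; FN_TimedNiagaraNotify_MaterialTransition is a timed anim notify state that temporarily swaps skeletal mesh materials for animated transition effects, optionally across player pawn part meshes.
; Begin stores original materials per notify instance ID, then applies dynamic copies, optionally using blend-mode-matched overrides with copied parameters, plus an optional overlay.
; Tick drives scalar parameters from blended or raw animation curves.
; End restores the original materials and overlay.
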(fn CreateMaterialTransitionDynamicMaterial
  (params (Material%20Source :type Material%20Interface%20Object%20Reference%20%28by%20ref%29) (Skel%20Mesh :type Skeletal%20Mesh%20Component%20Object%20Reference%20%28by%20ref%29))
  (bind _material_name (LocalVariables|GetMaterial%20Name))
  (bind _dyn_material (LocalVariables|GetDyn%20Material))
  (LocalVariables|SetMaterial%20Name (Utilities|String|StringToName (Utilities|GetDisplayName Material%20Source)))
  (bind _asmaterial_instance_dynamic (Utilities|Casting|CastToMaterialInstanceDynamic Material%20Source)
    (:then
      (Utilities|IsValid (Variables|/Script/Engine.MaterialInstance|GetParent _asmaterial_instance_dynamic)
        (:Is%20Valid
          (bind _returnvalue (Rendering|Material|CreateDynamicMaterialInstance Skel%20Mesh (Variables|/Script/Engine.MaterialInstance|GetParent _asmaterial_instance_dynamic) _material_name "None" :!class /Script/Engine.KismetMaterialLibrary))
          (LocalVariables|SetDyn%20Material _returnvalue))))
    (:CastFailed
      (bind _returnvalue_2 (Rendering|Material|CreateDynamicMaterialInstance Skel%20Mesh Material%20Source _material_name "None" :!class /Script/Engine.KismetMaterialLibrary))
      (LocalVariables|SetDyn%20Material _returnvalue_2)))
  (Utilities|IsValid _dyn_material
    (:Is%20Valid
      (Rendering|Material|CopyMaterialInstanceParameters _dyn_material Material%20Source false)
      (return (LocalVariables|GetDyn%20Material)))))

(fn GetOverlayDynamicMaterial
  (params (Skel%20Mesh :type Skeletal%20Mesh%20Component%20Object%20Reference))
  (bind _overlay_material (Variables|MaterialTransition|OverlayMaterial|GetOverlay%20Material))
  (if (Variables|MaterialTransition|OverlayMaterial|GetUse%20Overlay%20Material)
    (Utilities|IsValid _overlay_material
      (:Is%20Valid
        (LocalVariables|SetOverlay%20Name (Utilities|String|StringToName (Utilities|String|Append (Utilities|GetDisplayName _overlay_material) "_Overlay")))
        (bind _asmaterial_instance_dynamic (Utilities|Casting|CastToMaterialInstanceDynamic (Variables|MaterialTransition|OverlayMaterial|GetOverlay%20Material))
          (:then
            (bind _returnvalue (Rendering|Material|CreateDynamicMaterialInstance Skel%20Mesh (Variables|/Script/Engine.MaterialInstance|GetParent _asmaterial_instance_dynamic) (LocalVariables|GetOverlay%20Name) "None" :!class /Script/Engine.KismetMaterialLibrary))
            (return _returnvalue))
          (:CastFailed
            (bind _returnvalue_2 (Rendering|Material|CreateDynamicMaterialInstance Skel%20Mesh (Variables|MaterialTransition|OverlayMaterial|GetOverlay%20Material) (LocalVariables|GetOverlay%20Name) "None" :!class /Script/Engine.KismetMaterialLibrary))
            (return _returnvalue_2)))))
    (else
      (return 0))))

(fn GetOverrideMaterialFromMap
  (params (Material :type Material%20Interface%20Object%20Reference%20%28by%20ref%29) (Skel%20Mesh :type Skeletal%20Mesh%20Component%20Object%20Reference%20%28by%20ref%29))
  (bind _returnvalue (Rendering|Material|GetBlendMode Material))
  (bind (_override_material_map_opaque_2_e5397a2348f4fe40d0bf2ea59ff981e8 _override_material_map_masked_4_858a2caf4d02e8bfb6c886894ba50109 _override_material_map_translucentgreytransmittance_14_1662c11f4bc2aaeec63730be6fea3aba _override_material_map_additive_8_7b8994a04f9fdb0b8a062e95f6960257 _override_material_map_coloredtransmittanceonly_13_061b8bac436fd1f1a5b6ef91770d71f3 _override_material_map_alphacomposite_15_266d55184afc7c92f279969015363e6a _override_material_map_alphaholdout_16_b7a3066e4f6fca589c7dfd8c8dda12ad _override_material_map_translucentcoloredtransmittance_17_931e2c3744b9f627a8dc87a5ca0239f9) (Utilities|Struct|BreakStrucNiagaraNotifyMaterialTransitionMaterialOverrideMap (Variables|GetOverride%20Material%20Map_Opaque_2_E5397A2348F4FE40D0BF2EA59FF981E8)))
  (bind _returnvalue_2 (Utilities|Select :Index _returnvalue :BLEND_Opaque _override_material_map_opaque_2_e5397a2348f4fe40d0bf2ea59ff981e8 :BLEND_Masked _override_material_map_masked_4_858a2caf4d02e8bfb6c886894ba50109 :BLEND_Translucent _override_material_map_translucentgreytransmittance_14_1662c11f4bc2aaeec63730be6fea3aba :BLEND_Additive _override_material_map_additive_8_7b8994a04f9fdb0b8a062e95f6960257 :BLEND_Modulate _override_material_map_coloredtransmittanceonly_13_061b8bac436fd1f1a5b6ef91770d71f3 :BLEND_AlphaComposite _override_material_map_alphacomposite_15_266d55184afc7c92f279969015363e6a :BLEND_AlphaHoldout _override_material_map_alphaholdout_16_b7a3066e4f6fca589c7dfd8c8dda12ad :BLEND_TranslucentColoredTransmittance _override_material_map_translucentcoloredtransmittance_17_931e2c3744b9f627a8dc87a5ca0239f9))
  (bind _override_material_map_opaque_2_e5397a2348f4fe40d0bf2ea59ff981e8_2 (Utilities|Struct|BreakStrucNiagaraNotifyMaterialTransitionMaterialOverrideMap (Variables|GetOverride%20Material%20Map_Opaque_2_E5397A2348F4FE40D0BF2EA59FF981E8)))
  (bind _dynamic_material (CallFunction|CreateMaterialTransitionDynamicMaterial Material Skel%20Mesh))
  (LocalVariables|SetTemp%20Orig%20Dyn%20Material _dynamic_material)
  (Utilities|IsValid _returnvalue_2
    (:Is%20Valid
      (LocalVariables|SetTemp%20Selected%20Override%20Material _returnvalue_2))
    (:Is%20Not%20Valid
      (Utilities|IsValid _override_material_map_opaque_2_e5397a2348f4fe40d0bf2ea59ff981e8_2
        (:Is%20Valid
          (LocalVariables|SetTemp%20Selected%20Override%20Material _override_material_map_opaque_2_e5397a2348f4fe40d0bf2ea59ff981e8_2))
        (:Is%20Not%20Valid
          (LocalVariables|SetTemp%20Selected%20Override%20Material Material)))))
  (bind _dynamic_material_2 (CallFunction|CreateMaterialTransitionDynamicMaterial (LocalVariables|GetTemp%20Selected%20Override%20Material) Skel%20Mesh))
  (Rendering|Material|CopyMaterialInstanceParameters _dynamic_material_2 (LocalVariables|GetTemp%20Orig%20Dyn%20Material) false)
  (return _dynamic_material_2))

(fn Received_NotifyBegin
  (params (MeshComp :type Skeletal%20Mesh%20Component%20Object%20Reference) (Animation :type Anim%20Sequence%20Base%20Object%20Reference) (TotalDuration :type Float%20%28single-precision%29) (EventReference :type Anim%20Notify%20Event%20Reference%20Structure%20%28by%20ref%29))
  (bind _temp_material_dynamic (LocalVariables|GetTemp%20Material%20Dynamic))
  (bind _temp_collection_array_original (LocalVariables|GetTemp%20Collection%20Array%20Original))
  (bind _temp_collection_array_dyn (LocalVariables|GetTemp%20Collection%20Array%20Dyn))
  (|CallParentFunction MeshComp Animation TotalDuration EventReference)
  (LocalVariables|SetInstance%20ID (Utilities|Struct|BreakAnimNotifyEventReference EventReference))
  (Utilities|FlowControl|Sequence
    (:then_0
      ; Get all Skeletal Mesh Components
      (if (Variables|MaterialTransition|General|GetApply%20To%20Player)
        (Utilities|FlowControl|Sequence
          (:then_0
            (bind _asfort_player_pawn (Utilities|Casting|CastToFortPlayerPawn (Components|GetOwner MeshComp)))
            (bind _enumvalue (Utilities|Enum|ForEachEFortCustomPartType false)
              (:LoopBody
                (Utilities|IsValid (Customization|GetSkeletalMeshforPartType _asfort_player_pawn _enumvalue :!class /Script/FortniteGame.FortPlayerPawn)
                  (:Is%20Valid
                    (Utilities|Array|Add (LocalVariables|GetTemp%20Collection%20Array%20Original) (Utilities|Struct|MakeStrucNiagaraNotifyMaterialTransitionSkelMatMap (Customization|GetSkeletalMeshforPartType _asfort_player_pawn _enumvalue :!class /Script/FortniteGame.FortPlayerPawn)))
                    (Utilities|Array|Add (LocalVariables|GetTemp%20Collection%20Array%20Dyn) (Utilities|Struct|MakeStrucNiagaraNotifyMaterialTransitionSkelMatMap (Customization|GetSkeletalMeshforPartType _asfort_player_pawn _enumvalue :!class /Script/FortniteGame.FortPlayerPawn))))))))
          (:then_1
            _)))
      (Utilities|IsValid MeshComp
        (:Is%20Valid
          (Utilities|Array|Add (LocalVariables|GetTemp%20Collection%20Array%20Original) (Utilities|Struct|MakeStrucNiagaraNotifyMaterialTransitionSkelMatMap MeshComp))
          (Utilities|Array|Add (LocalVariables|GetTemp%20Collection%20Array%20Dyn) (Utilities|Struct|MakeStrucNiagaraNotifyMaterialTransitionSkelMatMap MeshComp)))))
    (:then_1
      (for _index (range (+ (Utilities|Array|LastIndex _temp_collection_array_original) 1))
        (bind (_output_skeletalmesh_18_afdad5ac4c82ead252ceb2a8f6ca7c09 _output_materiallist_17_842751334020934ec76af5bf60eb2af2 _output_overlaymaterial_24_7c4c5448416d30103df904bb9f67749f) (Utilities|Struct|BreakStrucNiagaraNotifyMaterialTransitionSkelMatMap (Utilities|Array|Get(aref) _temp_collection_array_original _index)))
        (Utilities|Array|AppendArray _output_materiallist_17_842751334020934ec76af5bf60eb2af2 (Rendering|Material|GetMaterials _output_skeletalmesh_18_afdad5ac4c82ead252ceb2a8f6ca7c09))
        (Variables|SetBy-RefVar _output_overlaymaterial_24_7c4c5448416d30103df904bb9f67749f (Variables|/Script/Engine.MeshComponent|GetOverlayMaterial _output_skeletalmesh_18_afdad5ac4c82ead252ceb2a8f6ca7c09))
        (bind _overlay_material_dynamic (CallFunction|GetOverlayDynamicMaterial _output_skeletalmesh_18_afdad5ac4c82ead252ceb2a8f6ca7c09))
        (bind (_output_skeletalmesh_18_afdad5ac4c82ead252ceb2a8f6ca7c09_2 _output_materiallist_17_842751334020934ec76af5bf60eb2af2_2 _output_overlaymaterial_24_7c4c5448416d30103df904bb9f67749f_2) (Utilities|Struct|BreakStrucNiagaraNotifyMaterialTransitionSkelMatMap (Utilities|Array|Get(aref) _temp_collection_array_dyn _index)))
        (Variables|SetBy-RefVar _output_overlaymaterial_24_7c4c5448416d30103df904bb9f67749f_2 _overlay_material_dynamic)
        (for _array_element (Rendering|Material|GetMaterials _output_skeletalmesh_18_afdad5ac4c82ead252ceb2a8f6ca7c09)
          (if (Variables|MaterialTransition|OverrideMaterial|GetUse%20Override%20Material)
            (bind _selected_material (CallFunction|GetOverrideMaterialfromMap _array_element _output_skeletalmesh_18_afdad5ac4c82ead252ceb2a8f6ca7c09))
            (LocalVariables|SetTemp%20Material%20Dynamic _selected_material)
            (else
              (bind _dynamic_material (CallFunction|CreateMaterialTransitionDynamicMaterial _array_element _output_skeletalmesh_18_afdad5ac4c82ead252ceb2a8f6ca7c09))
              (LocalVariables|SetTemp%20Material%20Dynamic _dynamic_material)))
          (Utilities|IsValid _temp_material_dynamic
            (:Is%20Valid
              (Utilities|Array|Add _output_materiallist_17_842751334020934ec76af5bf60eb2af2_2 _temp_material_dynamic))))
        (Utilities|FlowControl|Sequence
          (:then_0
            (Utilities|Map|Add (Variables|MaterialTransition|Internal|GetMaterial%20Collection%20) (LocalVariables|GetInstance%20ID) (Utilities|Struct|MakeStrucNiagaraNotifyMaterialTransitionCollection (LocalVariables|GetTemp%20Collection%20Array%20Original) (LocalVariables|GetTemp%20Collection%20Array%20Dyn))))
          (:then_1
            (for (_array_element_2 _array_index) _output_materiallist_17_842751334020934ec76af5bf60eb2af2_2
              (Rendering|Material|SetMaterial _output_skeletalmesh_18_afdad5ac4c82ead252ceb2a8f6ca7c09 _array_index _array_element_2)))
          (:then_2
            (if (Variables|MaterialTransition|OverlayMaterial|GetUse%20Overlay%20Material)
              (Rendering|Material|SetOverlayMaterial _output_skeletalmesh_18_afdad5ac4c82ead252ceb2a8f6ca7c09 _output_overlaymaterial_24_7c4c5448416d30103df904bb9f67749f_2 false 0 :!class /Script/Engine.MeshComponent))))))
    (:then_2
      (return true))))

(fn Received_NotifyEnd
  (params (MeshComp :type Skeletal%20Mesh%20Component%20Object%20Reference) (Animation :type Anim%20Sequence%20Base%20Object%20Reference) (EventReference :type Anim%20Notify%20Event%20Reference%20Structure%20%28by%20ref%29))
  (bind _notifyinstanceid (Utilities|Struct|BreakAnimNotifyEventReference EventReference))
  (bind _material_collection (Variables|MaterialTransition|Internal|GetMaterial%20Collection%20))
  (bind _value (Utilities|Map|Find _material_collection _notifyinstanceid))
  (bind _value_original_8_fc64bdcc4bfc2cb03b2a46a8e9f2fb0e (Utilities|Struct|BreakStrucNiagaraNotifyMaterialTransitionCollection _value))
  (|CallParentFunction MeshComp Animation EventReference)
  (for _array_element _value_original_8_fc64bdcc4bfc2cb03b2a46a8e9f2fb0e
    (bind (_array_element_skeletalmesh_18_afdad5ac4c82ead252ceb2a8f6ca7c09 _array_element_materiallist_17_842751334020934ec76af5bf60eb2af2 _array_element_overlaymaterial_24_7c4c5448416d30103df904bb9f67749f) (Utilities|Struct|BreakStrucNiagaraNotifyMaterialTransitionSkelMatMap _array_element))
    (for (_array_element_2 _array_index) _array_element_materiallist_17_842751334020934ec76af5bf60eb2af2
      (Rendering|Material|SetMaterial _array_element_skeletalmesh_18_afdad5ac4c82ead252ceb2a8f6ca7c09 _array_index _array_element_2))
    (Rendering|Material|SetOverlayMaterial _array_element_skeletalmesh_18_afdad5ac4c82ead252ceb2a8f6ca7c09 _array_element_overlaymaterial_24_7c4c5448416d30103df904bb9f67749f false 0 :!class /Script/Engine.MeshComponent))
  (Utilities|Map|Remove _material_collection _notifyinstanceid)
  (return true))

(fn Received_NotifyTick
  (params (MeshComp :type Skeletal%20Mesh%20Component%20Object%20Reference) (Animation :type Anim%20Sequence%20Base%20Object%20Reference) (FrameDeltaTime :type Float%20%28single-precision%29) (EventReference :type Anim%20Notify%20Event%20Reference%20Structure%20%28by%20ref%29))
  (bind _notifyinstanceid (Utilities|Struct|BreakAnimNotifyEventReference EventReference))
  (bind _material_collection (Variables|MaterialTransition|Internal|GetMaterial%20Collection%20))
  (bind _value (Utilities|Map|Find _material_collection _notifyinstanceid))
  (bind (_value_original_8_fc64bdcc4bfc2cb03b2a46a8e9f2fb0e _value_dynamic_9_c7097a7f4d847077b321479ce23a5628) (Utilities|Struct|BreakStrucNiagaraNotifyMaterialTransitionCollection _value))
  (bind _returnvalue (Utilities|Animation|Notifies|GetCurrentAnimationTime EventReference))
  (|CallParentFunction MeshComp Animation FrameDeltaTime EventReference)
  (for _array_element (Variables|MaterialTransition|General|GetMaterial%20Curves)
    (bind (_array_element_uservariablename_5_c3e793d64b2060301331ac805c0b1026 _array_element_animcurvename_6_457883094db38a3ef423919ede4d33c1) (Utilities|Struct|BreakAnimCurveParameterPair _array_element))
    (for _array_element_2 _value_dynamic_9_c7097a7f4d847077b321479ce23a5628
      (bind _array_element_skeletalmesh_18_afdad5ac4c82ead252ceb2a8f6ca7c09 (Utilities|Struct|BreakStrucNiagaraNotifyMaterialTransitionSkelMatMap _array_element_2))
      (Rendering|Material|SetScalarParameterValueonMaterials _array_element_skeletalmesh_18_afdad5ac4c82ead252ceb2a8f6ca7c09 _array_element_uservariablename_5_c3e793d64b2060301331ac805c0b1026 (Math|Float|SelectFloat (Animation|Curves|GetCurveValue (Components|SkeletalMesh|GetAnimInstance MeshComp) _array_element_animcurvename_6_457883094db38a3ef423919ede4d33c1) (Animation|Curves|GetUnblendedAnimationCurveValue Animation _returnvalue _array_element_animcurvename_6_457883094db38a3ef423919ede4d33c1) (Variables|MaterialTransition|General|GetApply%20Anim%20Blending)))))
  (if (Variables|MaterialTransition|OverlayMaterial|GetUse%20Overlay%20Material)
    (for _array_element_3 (Variables|MaterialTransition|OverlayMaterial|GetOverlay%20Material%20Curves)
      (bind (_array_element_uservariablename_5_c3e793d64b2060301331ac805c0b1026_2 _array_element_animcurvename_6_457883094db38a3ef423919ede4d33c1_2) (Utilities|Struct|BreakAnimCurveParameterPair _array_element_3))
      (for _array_element_4 _value_dynamic_9_c7097a7f4d847077b321479ce23a5628
        (bind (_array_element_skeletalmesh_18_afdad5ac4c82ead252ceb2a8f6ca7c09_2 _array_element_materiallist_17_842751334020934ec76af5bf60eb2af2 _array_element_overlaymaterial_24_7c4c5448416d30103df904bb9f67749f) (Utilities|Struct|BreakStrucNiagaraNotifyMaterialTransitionSkelMatMap _array_element_4))
        (Utilities|IsValid _array_element_overlaymaterial_24_7c4c5448416d30103df904bb9f67749f
          (:Is%20Valid
            (bind _asmaterial_instance_dynamic (Utilities|Casting|CastToMaterialInstanceDynamic _array_element_overlaymaterial_24_7c4c5448416d30103df904bb9f67749f))
            (Rendering|Material|SetScalarParameterValue _asmaterial_instance_dynamic _array_element_uservariablename_5_c3e793d64b2060301331ac805c0b1026_2 (Math|Float|SelectFloat (Animation|Curves|GetCurveValue (Components|SkeletalMesh|GetAnimInstance MeshComp) _array_element_animcurvename_6_457883094db38a3ef423919ede4d33c1_2) (Animation|Curves|GetUnblendedAnimationCurveValue Animation _returnvalue _array_element_animcurvename_6_457883094db38a3ef423919ede4d33c1_2) (Variables|Curves|GetUse%20Animation%20Blending)) :!class /Script/Engine.MaterialInstanceDynamic)))))
    (else
      _))
  (return true))
