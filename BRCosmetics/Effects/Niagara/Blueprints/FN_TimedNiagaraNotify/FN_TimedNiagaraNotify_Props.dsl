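; Timed Niagara anim notify state, filed as a props variant, that feeds Fortnite player pawn mesh and animation data into its spawned effect.
; On begin, it binds the pawn's skeletal mesh for a configured part type to a named Niagara user parameter, returning false if unavailable.
; Each tick, it copies configured curve values from the pawn mesh's anim instance, rather than the notify's own mesh, into Niagara float parameters.
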(fn Received_NotifyBegin
  (params (MeshComp :type Skeletal%20Mesh%20Component%20Object%20Reference) (Animation :type Anim%20Sequence%20Base%20Object%20Reference) (TotalDuration :type Float%20%28single-precision%29) (EventReference :type Anim%20Notify%20Event%20Reference%20Structure%20%28by%20ref%29))
  (bind _owner_skeleton_part_type (Variables|General|GetOwner%20Skeleton%20Part%20Type))
  (bind _returnvalue (|CallParentFunction MeshComp Animation TotalDuration EventReference))
  (if _returnvalue
    (bind _returnvalue_2 (AnimNotify|GetSpawnedEffect MeshComp))
    (bind _asniagara_particle_system_component (Utilities|Casting|CastToNiagaraComponent _returnvalue_2)
      (:then
        (LocalVariables|SetNiagaraSystem _asniagara_particle_system_component)
        (Utilities|FlowControl|Sequence
          (:then_0
            (bind _asfort_player_pawn (Utilities|Casting|CastToFortPlayerPawn (Components|GetOwner MeshComp))
              (:then
                (if (Utilities|IsValid (Customization|GetSkeletalMeshforPartType _asfort_player_pawn _owner_skeleton_part_type :!class /Script/FortniteGame.FortPlayerPawn) :!class /Script/Engine.KismetSystemLibrary)
                  (Niagara|SetNiagaraSkeletalMeshComponent (LocalVariables|GetNiagaraSystem) (Variables|General|GetOwner%20Skeleton%20User%20Param%20Name) (Customization|GetSkeletalMeshforPartType _asfort_player_pawn _owner_skeleton_part_type :!class /Script/FortniteGame.FortPlayerPawn))
                  (else
                    (return false))))
              (:CastFailed
                (return false))))
          (:then_1
            (return true))))
      (:CastFailed
        (return false)))
    (else
      (return false))))

(fn Received_NotifyTick
  (params (MeshComp :type Skeletal%20Mesh%20Component%20Object%20Reference) (Animation :type Anim%20Sequence%20Base%20Object%20Reference) (FrameDeltaTime :type Float%20%28single-precision%29) (EventReference :type Anim%20Notify%20Event%20Reference%20Structure%20%28by%20ref%29))
  (bind _returnvalue (|CallParentFunction MeshComp Animation FrameDeltaTime EventReference))
  (if _returnvalue
    (Utilities|IsValid MeshComp
      (:Is%20Valid
        (bind _returnvalue_2 (AnimNotify|GetSpawnedEffect MeshComp))
        (Utilities|IsValid _returnvalue_2
          (:Is%20Valid
            (bind _asniagara_particle_system_component (Utilities|Casting|CastToNiagaraComponent _returnvalue_2))
            (LocalVariables|SetNiagaraSystem _asniagara_particle_system_component)
            (bind _asfort_player_pawn (Utilities|Casting|CastToFortPlayerPawn (Components|GetOwner MeshComp)))
            ; Set Niagara Float parameters per Parent curve
            (for _array_element (Variables|Curves|GetOwnerAnimCurveParameterPair)
              (Niagara|SetNiagaraVariable(Float) (LocalVariables|GetNiagaraSystem) (Utilities|Struct|BreakAnimCurveParameterPair _array_element) (Animation|Curves|GetCurveValue (Components|SkeletalMesh|GetAnimInstance (Variables|/Script/Engine.Character|GetMesh _asfort_player_pawn)) (.AnimCurveName_6_457883094DB38A3EF423919EDE4D33C1 (Utilities|Struct|BreakAnimCurveParameterPair _array_element)))))
            (return true)))))
    (else
      (return false))))
