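; Timed Niagara anim notify state that feeds the owning Fortnite player pawn's stats into its spawned cosmetic effect.
; On begin, it runs the parent, then casts the spawned effect to a Niagara component and the mesh owner to a player pawn.
; An optional flag writes health and shield current, max and percent floats.
; Each Player Stats entry writes a cosmetic stat, with default, as an integer; tick only calls the parent, without refreshing values.
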(fn Received_NotifyBegin
  (params (MeshComp :type Skeletal%20Mesh%20Component%20Object%20Reference) (Animation :type Anim%20Sequence%20Base%20Object%20Reference) (TotalDuration :type Float%20%28single-precision%29) (EventReference :type Anim%20Notify%20Event%20Reference%20Structure%20%28by%20ref%29))
  (bind _niagarasystem (LocalVariables|GetNiagaraSystem))
  (bind _returnvalue (|CallParentFunction MeshComp Animation TotalDuration EventReference))
  (if _returnvalue
    (bind _returnvalue_2 (AnimNotify|GetSpawnedEffect MeshComp))
    (bind _asniagara_particle_system_component (Utilities|Casting|CastToNiagaraComponent _returnvalue_2)
      (:then
        (LocalVariables|SetNiagaraSystem _asniagara_particle_system_component)
        (Utilities|FlowControl|Sequence
          (:then_0
            (bind _asfort_player_pawn (Utilities|Casting|CastToFortPlayerPawn (Components|GetOwner MeshComp))
              (:then
                (if (Utilities|IsValid _asfort_player_pawn :!class /Script/Engine.KismetSystemLibrary)
                  (LocalVariables|SetFortPlayerPawn _asfort_player_pawn)
                  (Utilities|FlowControl|Sequence
                    (:then_0
                      (if (Variables|PlayerStats|GetSend%20Health%20&%20Shield%20Stats)
                        (bind _returnvalue_3 (Fortnite|Utilities|AttemptGetPlayerStateforPawn (LocalVariables|GetFortPlayerPawn)))
                        (bind _asfort_player_state_athena (Utilities|Casting|CastToFortPlayerStateAthena _returnvalue_3))
                        (bind (_health _healthmax _shield _shieldmax) (HealthSet|GetHealthandShields _asfort_player_state_athena))
                        (Niagara|SetNiagaraVariable(Float) _niagarasystem "Stat_Health" _health)
                        (Niagara|SetNiagaraVariable(Float) _niagarasystem "Stat_Max_Health" _healthmax)
                        (Niagara|SetNiagaraVariable(Float) _niagarasystem "Stat_Health_Percent" (/ _health _healthmax))
                        (Niagara|SetNiagaraVariable(Float) _niagarasystem "Stat_Shield" _shield)
                        (Niagara|SetNiagaraVariable(Float) _niagarasystem "Stat_Max_Shield" _shieldmax)
                        (Niagara|SetNiagaraVariable(Float) _niagarasystem "Stat_Shield_Percent" (/ _shield _shieldmax))))
                    (:then_1
                      (for _array_element (Variables|PlayerStats|GetPlayer%20Stats)
                        (bind _returnvalue_4 (Class|FortCosmeticFunctionLibrary|GetCosmeticStatValue (LocalVariables|GetFortPlayerPawn) (.StatTag_9_0FE2145A441406BCAE965F8B2FCD52B7 (Utilities|Struct|BreakStrucCosmeticStatParameter _array_element)) (.DefaultValue_12_E72DCD4D40B02F1775417C87630F62DD (Utilities|Struct|BreakStrucCosmeticStatParameter _array_element))))
                        (Niagara|SetNiagaraVariable(Int32) (LocalVariables|GetNiagaraSystem) (Utilities|Struct|BreakStrucCosmeticStatParameter _array_element) _returnvalue_4))))))
              (:CastFailed
                (return false))))
          (:then_1
            (return true))))
      (:CastFailed
        (return false)))
    (else
      (return false))))

(fn Received_NotifyTick
  (params (MeshComp :type Skeletal%20Mesh%20Component%20Object%20Reference) (Animation :type Anim%20Sequence%20Base%20Object%20Reference) (FrameDeltaTime :type Float%20%28single-precision%29) (EventReference :type Anim%20Notify%20Event%20Reference%20Structure%20%28by%20ref%29))
  (|CallParentFunction MeshComp Animation FrameDeltaTime EventReference))
